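# Animation Blueprint for the Blue Badass female hat cosmetic, making the hat follow the character's skeleton.
# It copies the pose from the attached parent mesh, then a Transform (Modify) Bone node replaces the C_nose_bridge bone's translation and rotation in parent bone space.
# The event graph checks the current LOD each update, so the bone override applies only at LOD 1.
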AnimBlueprint /BRCosmetics/Accessories/Hats/F_MED_BlueBadass/Meshes/F_MED_Blue_Badass_Hat_MESH_Skeleton_AnimBlueprint
target_skeleton=&/F_MED_Blue_Badass_Hat_MESH_Skeleton
graphs [
 AnimGraph {nodes [
   n1 = "Transform (Modify) Bone - Bone: C_nose_bridge"!(
    ComponentPose=Local To Component(LocalPose=Copy Pose From Mesh() {bUseAttachedParent=true})
    bAlphaBoolEnabled:bool=Get isLod_1() -> (isLod_1:bool)
    Translation="-0.0368,12,5.7674"
    Rotation="-1.6163,0.8789,-25.6438"
    Scale="1,1,1"
   ) -> (Pose) {
    alphaInputType=Bool
    rotationMode=BMM_Replace
    rotationSpace=BCS_ParentBoneSpace
    translationMode=BMM_Replace
    translationSpace=BCS_ParentBoneSpace
    alphaBoolBlend {blendOption=Linear}
    boneToModify {boneName=C_nose_bridge}
    rotation=<rot -1.6163 0.8789 -25.6438>
    translation=<-0.0368 12 5.7674>
   }
   Output Pose(Result=Component To Local(ComponentPose=@n1)) {layerGroup=None}
  ]}
 EventGraph {nodes [
   Event BlueprintUpdateAnimation()
   > Set isLod_1(isLod_1:bool="Equal (Integer)"(A=Get CurrentLOD(), B=1)) -> (Output_Get:bool)
  ]}
]
variables [isLod_1 {type=boolean}]
